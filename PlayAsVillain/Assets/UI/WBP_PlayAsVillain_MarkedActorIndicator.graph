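# Marked-actor indicator widget for PlayAsVillain, derived from AthenaMarkedActorIndicator, with large diamond icon, small indicator, pulsing borders and distance text.
# On construct it plays Anim_Intro, a brief scale pop on the root overlay.
# When marker data is set, it async-loads the display info's icon texture and writes it to the icon's dynamic material.
# The same function applies hard-coded blue and grey colors to icon and pulse materials, ignoring the marker's PrimaryColor.
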WidgetBlueprint /PlayAsVillain/Assets/UI/WBP_PlayAsVillain_MarkedActorIndicator
animations [Anim_Intro {
  binding_count=1
  length_seconds=0.7
  object=&self:Anim_Intro
  sequence {bindings [Overlay_Root {
     id=#1
     possessed_class=Overlay
     tracks [{
       type=MovieScene2DTransformTrack
       sections [{
         type=MovieScene2DTransformSection
         end_frame=14
         start_frame=0
         channels [
          Scale.X {type=MovieSceneScriptingFloatChannel, keys [[0,1],[11,1.5],[14,1]]}
          Scale.Y {type=MovieSceneScriptingFloatChannel, keys [[0,1],[11,1.5],[14,1]]}
         ]
        }]
      }]
    }]}
 }]
graphs [
 EventGraph {nodes [
   Event Construct()
   > PlayAnimation(InAnimation=Get Anim_Intro() -> (Anim_Intro:Widget Animation Object Reference))
   Event OnSetMarkerData!() -> (OutputDelegate, then, FortWorldMarkerData:"Fort World Marker Data Structure (by ref)") {is=title K2Node_Event}
   > n1 = Async Load Asset(Asset=Break Marked Actor Display Info!(MarkedActorDisplayInfo:"Marked Actor Display Info Structure (by ref)"=Get MarkerDisplay() -> (MarkerDisplay:Marked Actor Display Info Structure)) -> (
     DisplayName:Text
     Icon:Texture 2D Soft Object Reference
     CustomIndicatorClass:User Widget Soft Class Reference
     CustomDetailsClass:User Widget Soft Class Reference
     PrimaryColor:Linear Color Structure
     Sound:Sound Base Object Reference
     bAttachSound:bool=~
     ScreenClamping:EFortMarkedActorScreenClamping Enum
     SpecialCull:Bitmask Integer
     SpecialDotSizeThreshold:float=~
     SpecialDistance:float=~
     CustomTimeToLive:float=~
     AdditionalData:Marked Actor Display Info Additional Data Object Reference
    ) {is=title K2Node_BreakStruct}.Icon)
   n2 = Cast To Texture2D(execute=@n1.Completed, Object=@n1)
   > UpdateObjectiveIcon/Colors!(execute, self, Value:Texture Object Reference=@n2) -> (then)
  ]}
 UpdateObjectiveIcon/Colors {nodes [
   n1 = UpdateObjectiveIcon/Colors!() -> (then, Value:Texture Object Reference) {is=title K2Node_FunctionEntry}
   > n2 = GetDynamicMaterial(self=Get Image_ObjectiveIcon() -> (Image_ObjectiveIcon:Image Object Reference))
   > SetTextureParameterValue(self=@n2, ParameterName=Icon, Value=@n1)
   > SetVectorParameterValue(self=@n2, ParameterName=OutlineColor, Value=(R=0.0065,G=0.4678,B=0.7682,A=1))
   > SetVectorParameterValue(self=@n2, ParameterName=FillColor, Value=(R=0.3542,G=0.3542,B=0.3542,A=1))
   > SetScalarParameterValue(self=@n2, ParameterName=Opacity, Value=1)
   > n3 = GetDynamicMaterial(self=Get Image_Objective_BorderPulse() -> (Image_Objective_BorderPulse:Image Object Reference))
   > SetVectorParameterValue(self=@n3, ParameterName=Color, Value=(R=0.0065,G=0.4678,B=0.7682,A=1))
   > n4 = GetDynamicMaterial(self=Get Image_Objective_BorderPulse_Small() -> (Image_Objective_BorderPulse_Small:Image Object Reference))
   > SetVectorParameterValue(self=@n4, ParameterName=Color, Value=(R=0.0065,G=0.4678,B=0.7682,A=1))
   > SetScalarParameterValue(self=@n4, ParameterName=WidthEnd, Value=0.25)
   > SetScalarParameterValue(self=@n4, ParameterName=WidthStart, Value=0.25)
  ]}
]
variables []
widgets {
 info {inheritedWidgetCount=5, parentClass=/Script/FortniteUI.AthenaMarkedActorIndicator, rootWidgetClass=&Overlay, widgetCount=11}
 widgets [
  Overlay_Root {widgetClassPath=&Overlay, widget=Overlay_Root}
  Panel_LargeIndicator {
   widgetClassPath=&Overlay
   bInherited=true
   parent=Overlay_Root
   widget=Panel_LargeIndicator
   slot {class=OverlaySlot, properties {horizontalAlignment=HAlign_Center, verticalAlignment=VAlign_Center, padding=<ltrb 0 0 0 2>}}
  }
  Image_Objective_BorderPulse {
   widgetClassPath=&Image
   bIsVariable=true
   parent=Panel_LargeIndicator
   widget=Image_Objective_BorderPulse
   properties {
    brush {
     bIsDynamicallyLoaded=false
     drawAs=Image
     imageType=NoImage
     mirroring=NoMirror
     resourceName=None
     resourceObject=/Game/Athena/HUD/LTM/Infiltration/Materials/M_UI_GameplayDiamondPing_Pulse_Add_Inst
     tiling=NoTile
     imageSize=<84 84>
     margin=<ltrb 0 0 0 0>
     outlineSettings {
      bUseBrushTransparency=false
      roundingType=HalfHeightRadius
      width=0
      color {colorUseRule=UseColor_Specified, specifiedColor=<rgba 0 0 0 0>}
      cornerRadii=<0 0 0 0>
     }
     tintColor {colorUseRule=UseColor_Specified, specifiedColor=<rgba 0.0065 0.4678 0.7682 1>}
     uVRegion {bIsValid=false, max=<0 0>, min=<0 0>}
    }
    colorAndOpacity=<rgba 1 1 1 0.3>
   }
   slot {class=OverlaySlot, properties {horizontalAlignment=HAlign_Fill, verticalAlignment=VAlign_Fill}}
  }
  Image_ObjectiveIcon {
   widgetClassPath=&Image
   bIsVariable=true
   parent=Panel_LargeIndicator
   widget=Image_ObjectiveIcon
   properties {brush {
     bIsDynamicallyLoaded=false
     drawAs=Image
     imageType=NoImage
     mirroring=NoMirror
     resourceName=None
     resourceObject=/Game/Athena/HUD/Materials/M_UI_PingDiamond_Inst
     tiling=NoTile
     imageSize=<72 72>
     margin=<ltrb 0 0 0 0>
     outlineSettings {
      bUseBrushTransparency=false
      roundingType=HalfHeightRadius
      width=0
      color {colorUseRule=UseColor_Specified, specifiedColor=<rgba 0 0 0 0>}
      cornerRadii=<0 0 0 0>
     }
     tintColor {colorUseRule=UseColor_Specified, specifiedColor=<rgba 1 1 1 1>}
     uVRegion {bIsValid=false, max=<0 0>, min=<0 0>}
    }}
   slot {class=OverlaySlot, properties {horizontalAlignment=HAlign_Center, verticalAlignment=VAlign_Center}}
  }
  Panel_SmallIndicator {
   widgetClassPath=&Overlay
   bInherited=true
   parent=Overlay_Root
   widget=Panel_SmallIndicator
   properties {renderTransform {angle=0, scale=<1 1>, shear=<0 0>, translation=<0 10>}}
   slot {class=OverlaySlot, properties {horizontalAlignment=HAlign_Center, verticalAlignment=VAlign_Center, padding=<ltrb 0 0 0 2>}}
  }
  Image_Objective_BorderPulse_Small {
   widgetClassPath=&Image
   bIsVariable=true
   parent=Panel_SmallIndicator
   widget=Image_Objective_BorderPulse_Small
   properties {
    brush {
     bIsDynamicallyLoaded=false
     drawAs=Image
     imageType=NoImage
     mirroring=NoMirror
     resourceName=None
     resourceObject=&M_UI_GameplayDiamondPing_Pulse_Add_Inst
     tiling=NoTile
     imageSize=<16 16>
     margin=<ltrb 0 0 0 0>
     outlineSettings {
      bUseBrushTransparency=false
      roundingType=HalfHeightRadius
      width=0
      color {colorUseRule=UseColor_Specified, specifiedColor=<rgba 0 0 0 0>}
      cornerRadii=<0 0 0 0>
     }
     tintColor {colorUseRule=UseColor_Specified, specifiedColor=<rgba 0.0065 0.4678 0.7682 1>}
     uVRegion {bIsValid=false, max=<0 0>, min=<0 0>}
    }
    colorAndOpacity=<rgba 1 1 1 0.3>
    renderTransform {angle=0, scale=<1 1>, shear=<0 0>, translation=<0 14>}
    renderTransformPivot=<5.5 0.5>
   }
   slot {class=OverlaySlot, properties {horizontalAlignment=HAlign_Fill, verticalAlignment=VAlign_Fill}}
  }
  Image_SmallIndicator {
   widgetClassPath=&Image
   parent=Panel_SmallIndicator
   widget=Image_SmallIndicator
   properties {brush {
     bIsDynamicallyLoaded=false
     drawAs=Image
     imageType=NoImage
     mirroring=NoMirror
     resourceName=None
     resourceObject=/Game/Athena/HUD/Materials/M_PingIndicator
     tiling=NoTile
     imageSize=<32 32>
     margin=<ltrb 0 0 0 0>
     outlineSettings {
      bUseBrushTransparency=false
      roundingType=HalfHeightRadius
      width=0
      color {colorUseRule=UseColor_Specified, specifiedColor=<rgba 0 0 0 0>}
      cornerRadii=<0 0 0 0>
     }
     tintColor {colorUseRule=UseColor_Specified, specifiedColor=<rgba 0.0065 0.4678 0.7682 1>}
     uVRegion {bIsValid=false, max=<0 0>, min=<0 0>}
    }}
   slot {class=OverlaySlot, properties {horizontalAlignment=HAlign_Center, verticalAlignment=VAlign_Bottom}}
  }
  VisAttachment_Distance {
   widgetClassPath=&FortVisualAttachment
   parent=Overlay_Root
   widget=VisAttachment_Distance
   properties {renderTransform {angle=0, scale=<1 1>, shear=<0 0>, translation=<20 10>}}
   slot {class=OverlaySlot, properties {horizontalAlignment=HAlign_Center, verticalAlignment=VAlign_Center}}
  }
  Text_Distance {
   widgetClassPath=&CommonTextBlock
   bInherited=true
   parent=VisAttachment_Distance
   widget=Text_Distance
   properties {
    text=100m
    font {
     bForceMonospaced=false
     bMaterialIsStencil=false
     fontMaterial=null
     fontObject=&Burbank
     letterSpacing=0
     monospacedWidth=1
     size=14
     skewAmount=0.2
     typefaceFontName=BurbankRegular-Black
     outlineSettings {
      bApplyOutlineToDropShadows=false
      bMiteredCorners=false
      bSeparateFillAlpha=false
      outlineMaterial=null
      outlineSize=1
      outlineColor=<rgba 0 0 0 0.6>
     }
    }
    shadowColorAndOpacity=<rgba 0 0 0 1>
    shadowOffset=<3 3>
   }
   slot {class=SizeBoxSlot}
  }
  Switcher_IndicatorSize {widgetClassPath=&WidgetSwitcher, bInherited=true, widget=null}
  TutorialHighlight {widgetClassPath=&TutorialHighlight}
 ]
}
