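# Translucent UI material for the PlayAsVillain floating overshield indicator, drawn from MSDF textures with anti-aliased edges.
# Progress and DeltaProgress drive vertical step masks for the fill and its delta highlight, tinted by a two-color gradient and a delta color.
# The shield shape blends between backing and break textures, with threshold offsets varying edge size.
# A lightning bolt is subtracted from the shape as a cutout, with parameters for scale and overall opacity.
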Material /PlayAsVillain/UI/M_UI_PlayAsVillain_FloatingOvershield
graph {
 properties {bAutomaticallySetUsageInEditor=false, blendMode=BLEND_Translucent, materialDomain=MD_UI, usedShadingModels=Default Lit}
 nodes [
  n1 = VectorParameter() {parameterName=DeltaProgressColor, defaultValue=<rgba 0.327 0.9184 1 1>}
  n2 = NamedRerouteUsage()
  n3 = VectorParameter() {parameterName=UpperColor, defaultValue=<rgba 0.0331 0.855 1 1>}
  n4 = VectorParameter() {parameterName=LowerCOlor, defaultValue=<rgba 0.0272 0.7018 0.8208 1>}
  n5 = NamedRerouteUsage()
  n6 = MaterialFunctionCall(Texture Scale=NamedRerouteUsage()) {materialFunction=&ScaleUVsByCenter}
  n7 = MaterialFunctionCall!(MSDF RGB=LinearInterpolate(
    A=TextureSample(UVs=@n6.UVs) {texture=&/MSDF_UI_PlayAsBoss_FloatingOverShield_Backing}.RGB
    B=TextureSample(UVs=@n6.UVs) {texture=&/MSDF_UI_PlayAsBoss_FloatingOvershield_Break}.RGB
    Alpha=NamedRerouteUsage()
   ) {constAlpha=1}) {materialFunction=/FortUILibrary/MaterialFunctions/Shapes/Modifiers/MF_UI_MSDFtoSDF}
  n8 = LinearInterpolate(A=Constant() {r=0.1}, Alpha=NamedRerouteUsage()) {constB=0.25}
  n9 = MaterialFunctionCall(DistanceField=@n7.SDF, Threshold=@n8) {materialFunction=&MF_UI_AntiAliasing}
  Multiply(
   A=LinearInterpolate(A=Multiply(A=@n1.RGB, B=@n2), B=LinearInterpolate(A=@n3.RGB, B=@n4.RGB, Alpha=@n5), Alpha=NamedRerouteUsage()) {constA=1}
   B=@n9.Outer
  )
  n10 = NamedRerouteUsage()
  n11 = NamedRerouteUsage()
  n12 = Multiply(A=@n9.Outer, B=Saturate(None=Add(A=@n10, B=@n11)))
  n13 = MaterialFunctionCall(DistanceField=@n7.SDF, Threshold=Add(A=@n8) {constB=-0.035}) {materialFunction=&MF_UI_AntiAliasing}
  n14 = MaterialFunctionCall(UVs=TextureCoordinate(), Texture Scale=Multiply(A=NamedRerouteUsage(), B=NamedRerouteUsage())) {materialFunction=&ScaleUVsByCenter}
  n15 = Constant2Vector() {g=0.05, r=-0.03}
  n16 = MaterialFunctionCall(MSDF RGB=TextureSample(UVs=Add(A=@n14.UVs, B=@n15)) {texture=&/MSDF_UI_PlayAsVillain_FloatingOvershield_Lightining}.RGB) {materialFunction=&MF_UI_MSDFtoSDF}
  n17 = Constant() {r=0.125}
  Multiply(
   A=Subtract(
    A=LinearInterpolate(A=@n12, B=@n13.Outer)
    B=MaterialFunctionCall(DistanceField=@n16.SDF, Threshold=@n17) {materialFunction=&MF_UI_AntiAliasing}.Outer
   )
   B=NamedRerouteUsage()
  )
  NamedRerouteDeclaration(None=ScalarParameter() {defaultValue=1.5, parameterName=MasterScale}) {name=Scale, nodeColor=<rgba 1 0 0.6797 1>}
  NamedRerouteDeclaration(None=ScalarParameter() {defaultValue=0.5, parameterName=Progress}) {name=Progress, nodeColor=<rgba 0 1 1 1>}
  NamedRerouteDeclaration(None=ScalarParameter() {defaultValue=0.45, parameterName=BoltScale}) {name=BoltScale, nodeColor=<rgba 1 0.6797 0 1>}
  n18 = ComponentMask(None=TextureCoordinate()) {g=true}
  NamedRerouteDeclaration(None=@n18) {name=GradientUVs, nodeColor=<rgba 1 0.4219 0 1>}
  NamedRerouteDeclaration(None=ScalarParameter() {defaultValue=0.9, parameterName=MasterOpacity}) {name=MasterOpacity, nodeColor=<rgba 0.1016 0 1 1>}
  NamedRerouteDeclaration(None=Step(Y=OneMinus(None=@n18), X=NamedRerouteUsage())) {name=ProgressMask, nodeColor=<rgba 0.2187 1 0 1>}
  NamedRerouteDeclaration(None=Step(Y=NamedRerouteUsage()) {constX=0, constY=0.5}) {name=ProgressStep, nodeColor=<rgba 1 0 0.5391 1>}
  NamedRerouteDeclaration(None=ScalarParameter() {defaultValue=0.6, parameterName=DeltaProgress}) {name=DeltaProgress, nodeColor=<rgba 0.2656 0 1 1>}
  n19 = ComponentMask(None=TextureCoordinate()) {g=true}
  NamedRerouteDeclaration(None=Step(Y=OneMinus(None=@n19), X=NamedRerouteUsage())) {name=DeltaProgressMask, nodeColor=<rgba 1 0.6563 0 1>}
 ]
}
