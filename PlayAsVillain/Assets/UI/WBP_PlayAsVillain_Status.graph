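# PlayAsVillain boss status widget, derived from NPCStatusWidget_AIBoss, showing a boss icon, settings-filtered player name, health bar and floating overshield.
# Overshield updates drive material parameters instantly or with a timer-stepped trailing delta, then call EventShowStatus.
# Visibility changes are gated on actual change, then either snap or play the opacity fade forward or reversed, based on SetInstant.
# SetColorOnDBNO swaps health and delta bar material colors for the downed state, previewable at design time.
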WidgetBlueprint /PlayAsVillain/Assets/UI/WBP_PlayAsVillain_Status
animations [Anim_ShowOverShield {
  binding_count=1
  length_seconds=0.25
  object=&self:Anim_ShowOverShield
  sequence {bindings [Image_Overshield {
     id=#1
     possessed_class=Image
     tracks [{
       type=MovieSceneFloatTrack
       display_name=Render Opacity
       sections [{type=MovieSceneFloatSection, end_frame=5, start_frame=0, channels [None {type=MovieSceneScriptingFloatChannel, keys [[0,0],[5,1]]}]}]
      }]
    }]}
 }]
graphs [
 EventGraph {nodes [
   n1 = SetPercentInstant!() -> (OutputDelegate, then, Value:double=~) {is=title K2Node_CustomEvent}
   > SetScalarParameterValue(self=Get OvershieldMID() -> (OvershieldMID:Material Instance Dynamic Object Reference), ParameterName=Progress, Value=@n1.Value)
   > SetScalarParameterValue(self=Get OvershieldMID(), ParameterName=DeltaProgress, Value=@n1.Value)
   n2 = EventOvershieldUpdate!() -> (OutputDelegate, then, overshield:double=~, maxOvershield:double=~, UpdateInstant:bool=~) {is=title K2Node_CustomEvent}
   > Set CurrentOvershieldPercent(CurrentOvershieldPercent:double=@n2.overshield) -> (Output_Get:double)
   > n3 = Branch(Condition=@n2.UpdateInstant)
   n4 = Get CurrentOvershieldPercent() -> (CurrentOvershieldPercent:double)
   > n5 = SetPercentInstant!(execute, self, Value:double=@n4) -> (then) {is=title K2Node_CallFunction}
   EventShowStatus(execute=[@n5.then, @n6.then], self:NPCStatus Widget AIBoss Object Reference)
   n6 = SetPercent!(execute=@n3.else, self, Value:double=@n4) -> (then)
   n7 = SetPercent!() -> (OutputDelegate, then, Value:double=~) {is=title K2Node_CustomEvent}
   > SetScalarParameterValue(self=Get OvershieldMID(), ParameterName=Progress, Value=@n7.Value)
   > SetDeltaProgressWithAnimation!(execute, self, CurrentProgressValue:double=@n7.Value) -> (then)
   n8 = SetOvershieldVisbility!() -> (OutputDelegate, then, IsVisible:bool=~, SetInstant:bool=~) {is=title K2Node_CustomEvent}
   > n9 = Delta Gate Wildcard!(execute:Exec, Variable:"Boolean (by ref)"=Get IsOverShieldVisible() -> (IsOverShieldVisible:bool), New Value:bool=@n8.IsVisible) -> (Success:Exec, Updated Variable:"Boolean (by ref)", Previous Variable:bool=~) {is=title K2Node_MacroInstance}
   > n10 = Branch(Condition=@n8.SetInstant)
   > n11 = Branch(Condition=@n9)
   > SetVisibility(self=Get Image_Overshield() -> (Image_Overshield:Image Object Reference), InVisibility=HitTestInvisible)
   > Play Animation with Finished event!(
    execute:Exec
    Widget:User Widget Object Reference=Self-Reference()
    InAnimation:Widget Animation Object Reference=Get Anim_ShowOverShield() -> (Anim_ShowOverShield:Widget Animation Object Reference)
    StartAtTime:float=0
    NumLoopsToPlay:int=1
    PlayMode:EUMGSequencePlayMode Enum
    PlaybackSpeed:float=1
   ) -> (then:Exec, Result:Widget Animation Handle Structure, Finished:Exec) {is=title K2Node_PlayAnimation2}
   n12 = Play Animation with Finished event(
    execute=@n11.else
    Widget=Self-Reference()
    InAnimation=Get Anim_ShowOverShield()
    StartAtTime=0
    NumLoopsToPlay=1
    PlayMode=Reverse
    PlaybackSpeed=1
   )
   SetVisibility(execute=@n12.Finished, self=Get Image_Overshield(), InVisibility=Collapsed)
   SetVisibility(
    execute=@n10.else
    self=Get Image_Overshield()
    InVisibility=Select(Option 0=Collapsed, Option 1=HitTestInvisible, Index=Get IsOverShieldVisible())
   )
   n13 = Event PreConstruct()
   > Branch(Condition=@n13.IsDesignTime)
   n14 = Get Bar_Health() -> (Bar_Health:NPCStatus Widget Bar Object Reference)
   > n15 = GetDynamicMaterial(self=Get ProgressBar!(self:NPCStatus Widget Bar Object Reference=@n14) -> (ProgressBar:Image Object Reference))
   > Set HealthBarMID(HealthBarMID:Material Instance Dynamic Object Reference=@n15) -> (Output_Get:Material Instance Dynamic Object Reference)
   > n16 = GetDynamicMaterial(self=Get ProgressBar_Delta!(self:NPCStatus Widget Bar Object Reference=@n14) -> (ProgressBar_Delta:Image Object Reference))
   > Set DeltaBarMID(DeltaBarMID:Material Instance Dynamic Object Reference=@n16) -> (Output_Get:Material Instance Dynamic Object Reference)
   > SetColorOnDBNO!(execute, self, IsDBNO:bool=Get DesignTestDBNOColors() -> (DesignTestDBNOColors:bool)) -> (then)
   Event Construct()
   n17 = Get Bar_Health()
   > n18 = GetDynamicMaterial(self=Get ProgressBar(self=@n17))
   > Set HealthBarMID(HealthBarMID=@n18)
   > n19 = GetDynamicMaterial(self=Get ProgressBar_Delta(self=@n17))
   > Set DeltaBarMID(DeltaBarMID=@n19)
   > n20 = GetDynamicMaterial(self=Get Image_Overshield())
   > Set OvershieldMID(OvershieldMID:Material Instance Dynamic Object Reference=@n20) -> (Output_Get:Material Instance Dynamic Object Reference)
   > Set Visibility(Visibility=HitTestInvisible, self=Get NamedSlot_Name() -> (NamedSlot_Name:Named Slot Object Reference))
  ]}
 HandleDeltaProgressInterpolation {nodes [
   HandleDeltaProgressInterpolation() {is=title K2Node_FunctionEntry}
   > n1 = Set PreviousOvershieldDeltaProgressValue(PreviousOvershieldDeltaProgressValue:double=float - float(A=Get PreviousOvershieldDeltaProgressValue() -> (PreviousOvershieldDeltaProgressValue:double), B=0.015)) -> (Output_Get:double)
   > n2 = Branch(Condition="float <= float"(A=@n1, B=Get CurrentOvershieldPercent()) -> (ReturnValue=~))
   > Clear and Invalidate Timer by Handle(Handle=Get TimerOvershieldDeltaAnimateProgressUpdate() -> (TimerOvershieldDeltaAnimateProgressUpdate:Timer Handle Structure))
   > SetScalarParameterValue(
    self=Get OvershieldMID()
    ParameterName=DeltaProgress
    Value=Get CurrentOvershieldDeltaProgressValue() -> (CurrentOvershieldDeltaProgressValue:double)
   )
   > Set PreviousOvershieldDeltaProgressValue(PreviousOvershieldDeltaProgressValue=Get CurrentOvershieldDeltaProgressValue())
   SetScalarParameterValue(execute=@n2.else, self=Get OvershieldMID(), ParameterName=DeltaProgress, Value=Get PreviousOvershieldDeltaProgressValue())
  ]}
 SetColorOnDBNO {nodes [
   SetColorOnDBNO!() -> (then, IsDBNO:bool=~) {is=title K2Node_FunctionEntry}
   > SetVectorParameterValue(
    self=Get HealthBarMID() -> (HealthBarMID:Material Instance Dynamic Object Reference)
    ParameterName=ColorA
    Value=Select(
     Option 0:Linear Color Structure=Get FillColorA() -> (FillColorA:Linear Color Structure)
     Option 1:Linear Color Structure=Get DBNOFillColorA() -> (DBNOFillColorA:Linear Color Structure)
     Index=Get IsDBNO() -> (IsDBNO:bool)
    ) -> (ReturnValue:Linear Color Structure)
   )
   > SetVectorParameterValue(
    self=Get HealthBarMID()
    ParameterName=ColorB
    Value=Select(
     Option 0=Get Fill Color B() -> (Fill Color B:Linear Color Structure)
     Option 1=Get DBNOFillColorB() -> (DBNOFillColorB:Linear Color Structure)
     Index=Get IsDBNO()
    )
   )
   n1 = Select(
    Option 0=Get DeltaColor() -> (DeltaColor:Linear Color Structure)
    Option 1=Get DBNODeltaColor() -> (DBNODeltaColor:Linear Color Structure)
    Index=Get IsDBNO()
   )
   > SetVectorParameterValue(self=Get DeltaBarMID() -> (DeltaBarMID:Material Instance Dynamic Object Reference), ParameterName=ColorA, Value=@n1)
   > SetVectorParameterValue(self=Get DeltaBarMID(), ParameterName=ColorB, Value=@n1)
  ]}
 SetDeltaProgressWithAnimation {nodes [
   n1 = SetDeltaProgressWithAnimation!() -> (then, CurrentProgressValue:double=~) {is=title K2Node_FunctionEntry}
   > Set CurrentOvershieldDeltaProgressValue(CurrentOvershieldDeltaProgressValue:double=@n1) -> (Output_Get:double)
   > n2 = Set Timer by Event(
    Delegate=Create Event()
    Time=Get OvershieldDeltaProgressUpdateRate() -> (OvershieldDeltaProgressUpdateRate:double)
    bLooping=true
    InitialStartDelay=0.05
   )
   > Set TimerOvershieldDeltaAnimateProgressUpdate(TimerOvershieldDeltaAnimateProgressUpdate:Timer Handle Structure=@n2) -> (Output_Get:Timer Handle Structure)
  ]}
 SetPlayerName {nodes [
   SetPlayerName!() -> (then, OwningActor:Actor Object Reference) {is=title K2Node_FunctionEntry}
   n1 = GetSettingsFilteredPlayerNameSafe!(
    LocalPlayerController:Player Controller Object Reference=GetOwningPlayer()
    Actor:Actor Object Reference=Get OwningActor() -> (OwningActor:Actor Object Reference)
   ) -> (ReturnValue:String)
   > n2 = Branch(Condition=IsEmpty(InString=@n1))
   > Set Visibility(Visibility=Collapsed, self=Get NamedSlot_Name())
   Set Text(execute=@n2.else, Text="To Text (String)"(InString=@n1), self=Get Text_Name() -> (Text_Name:Common Text Object Reference))
  ]}
]
variables [
 CurrentOvershieldDeltaProgressValue {type=number}
 CurrentOvershieldPercent {type=number}
 DBNODeltaColor {type=LinearColor, default=<rgba 1 0.099 0.1779 1>}
 DBNOFillColorA {type=LinearColor, default=<rgba 0.3185 0.0086 0.0666 1>}
 DBNOFillColorB {type=LinearColor, default=<rgba 0.6038 0.0052 0.0908 1>}
 DeltaBarMID {type=MaterialInstanceDynamic}
 DeltaColor {type=LinearColor, default=<rgba 1 0.6795 0.1119 1>}
 DesignTestDBNOColors {type=boolean}
 Fill Color B {type=LinearColor, default=<rgba 0.9301 0.2051 0.0168 1>}
 FillColorA {type=LinearColor, default=<rgba 0.9301 0.0865 0.0185 1>}
 HealthBarMID {type=MaterialInstanceDynamic}
 IsOverShieldVisible {type=boolean}
 OvershieldDeltaProgressUpdateRate {type=number, default=0.03}
 OvershieldMID {type=MaterialInstanceDynamic}
 PreviousOvershieldDeltaProgressValue {type=number}
 TimerOvershieldDeltaAnimateProgressUpdate {type=TimerHandle}
]
widgets {
 info {
  inheritedWidgetCount=0
  namedSlotCount=2
  parentClass=/Game/Athena/HUD/Actors/NPCStatusWidget_AIBoss.NPCStatusWidget_AIBoss_C
  rootWidgetClass=null
  widgetCount=6
 }
 widgets [
  CommonTextBlock_35 {
   widgetClassPath=&CommonTextBlock
   widget=CommonTextBlock_35
   properties {
    style=&TextStyle-Base_C
    text=Text Block
    font {
     bForceMonospaced=false
     bMaterialIsStencil=false
     fontMaterial=null
     fontObject=&NotoSans
     letterSpacing=0
     monospacedWidth=1
     size=24
     skewAmount=0
     typefaceFontName=NotoSans-Regular
     outlineSettings {
      bApplyOutlineToDropShadows=false
      bMiteredCorners=false
      bSeparateFillAlpha=false
      outlineMaterial=null
      outlineSize=0
      outlineColor=<rgba 0 0 0 1>
     }
    }
    shadowOffset=<0 0>
   }
  }
  Overlay_Name {widgetClassPath=&Overlay, widget=Overlay_Name}
  HorizontalBox_1 {
   widgetClassPath=&HorizontalBox
   parent=Overlay_Name
   widget=HorizontalBox_1
   slot {class=OverlaySlot, properties {horizontalAlignment=HAlign_Center, verticalAlignment=VAlign_Center, padding=<ltrb 0 0 0 10>}}
  }
  Image_BossIcon {
   widgetClassPath=&Image
   parent=HorizontalBox_1
   widget=Image_BossIcon
   properties {
    brush {
     bIsDynamicallyLoaded=false
     drawAs=Image
     imageType=FullColor
     mirroring=NoMirror
     resourceName=None
     resourceObject=/PlayAsVillain/UI/T_UI_Icon__Minimap_PlayAsBoss
     tiling=NoTile
     imageSize=<25 25>
     margin=<ltrb 0 0 0 0>
     outlineSettings {
      bUseBrushTransparency=false
      roundingType=HalfHeightRadius
      width=0
      color {colorUseRule=UseColor_Specified, specifiedColor=<rgba 0 0 0 0>}
      cornerRadii=<0 0 0 0>
     }
     tintColor {colorUseRule=UseColor_Specified, specifiedColor=<rgba 1 1 1 1>}
     uVRegion {bIsValid=false, max=<0 0>, min=<0 0>}
    }
    renderTransform {angle=0, scale=<1.4 1.4>, shear=<0 0>, translation=<0 0>}
   }
   slot {class=HorizontalBoxSlot, properties {horizontalAlignment=HAlign_Center, verticalAlignment=VAlign_Center, padding=<ltrb 0 -3 2 0>}}
  }
  Text_Name {
   widgetClassPath=&CommonTextBlock
   bIsVariable=true
   parent=HorizontalBox_1
   widget=Text_Name
   properties {
    scrollStyle=&TextScrollStyle-Base_C
    text=PlayerName
    font {
     bForceMonospaced=false
     bMaterialIsStencil=false
     fontMaterial=null
     fontObject=&Burbank
     letterSpacing=0
     monospacedWidth=1
     size=14
     skewAmount=0
     typefaceFontName=BurbankRegular-Black
     outlineSettings {
      bApplyOutlineToDropShadows=true
      bMiteredCorners=false
      bSeparateFillAlpha=false
      outlineMaterial=null
      outlineSize=1
      outlineColor=<rgba 0 0 0 0.75>
     }
    }
    shadowColorAndOpacity=<rgba 0 0 0 1>
   }
   slot {class=HorizontalBoxSlot}
  }
  Image_Overshield {
   widgetClassPath=&Image
   bIsVariable=true
   parent=Overlay_Name
   widget=Image_Overshield
   properties {
    brush {
     bIsDynamicallyLoaded=false
     drawAs=Image
     imageType=FullColor
     mirroring=NoMirror
     resourceName=None
     resourceObject=/PlayAsVillain/UI/M_UI_PlayAsVillain_FloatingOvershield
     tiling=NoTile
     imageSize=<32 32>
     margin=<ltrb 0 0 0 0>
     outlineSettings {
      bUseBrushTransparency=false
      roundingType=HalfHeightRadius
      width=0
      color {colorUseRule=UseColor_Specified, specifiedColor=<rgba 0 0 0 0>}
      cornerRadii=<0 0 0 0>
     }
     tintColor {colorUseRule=UseColor_Specified, specifiedColor=<rgba 1 1 1 1>}
     uVRegion {bIsValid=false, max=<0 0>, min=<0 0>}
    }
    renderTransform {angle=0, scale=<1.25 1.25>, shear=<0 0>, translation=<45 25>}
   }
   slot {class=OverlaySlot, properties {horizontalAlignment=HAlign_Right}}
  }
 ]
}
